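; Lazy river waterpark VO gate that plays a different voice line depending on which side the local player enters its audio gameplay volume from.
; When the volume toggles on, the dot product of the pawn's normalized velocity and SpawnLocation's forward vector selects the front or rear sound.
; If the pawn is locally viewed and the sound is valid, it spawns attached to SpawnLocation and is stored as PlayingAudio.
; Toggling off does nothing.
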(event OnToggledOn(AudioGameplayVolumeToggle)
  (bind _returnvalue (Fortnite|GetFirstLocalFortPlayerController true))
  (bind _returnvalue_2 (Game|Player|GetPlayerPawn _returnvalue))
  (bind _charactermovement (Variables|/Script/Engine.Character|GetCharacterMovement _returnvalue_2))
  (bind _velocity (Variables|/Script/Engine.MovementComponent|GetVelocity _charactermovement))
  (bind _returnvalue_3 (Math|Vector|Normalize _velocity))
  (bind _spawnlocation (Variables|Default|GetSpawnLocation))
  (bind _returnvalue_4 (Transformation|GetForwardVector _spawnlocation))
  (bind _returnvalue_5 (Math|Vector|DotProduct _returnvalue_4 _returnvalue_3))
  (bind _returnvalue_6 (> _returnvalue_5 0))
  (bind _vo_toplay_onentryfromfront (Variables|Default|GetVO_ToPlay_OnEntryFromFront))
  (bind _vo_toplay_onentryfromrear (Variables|Default|GetVO_ToPlay_OnEntryFromRear))
  (bind _returnvalue_7 (select _returnvalue_6 _vo_toplay_onentryfromfront _vo_toplay_onentryfromrear))
  ;[0.5751 0.1986 0.8854 0.5] Entering either side of the volume
  (Utilities|IsValid _returnvalue
    (:Is%20Valid
      (Utilities|IsValid _returnvalue_2
        (:Is%20Valid
          (if (Pawn|IsLocallyViewed _returnvalue_2)
            (Utilities|IsValid _returnvalue_7
              (:Is%20Valid
                (bind _returnvalue_8 (Audio|SpawnSoundAttached _returnvalue_7 (Variables|Default|GetSpawnLocation)))
                (Variables|Default|SetPlayingAudio _returnvalue_8)))))))))

(event OnToggledOff(AudioGameplayVolumeToggle))

(fn ConstructionScript)
