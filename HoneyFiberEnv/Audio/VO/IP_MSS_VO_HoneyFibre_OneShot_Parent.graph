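# Parent one-shot mono MetaSound for HoneyFibre NPC voice-over, with an empty wave array input and AwesomeO statue attenuation.
# Changes in time since last played trigger comparisons that gate playback with a cooldown of about ten seconds, rather than using OnPlay.
# A passing check prints that time and plays a random wave, whose end finishes the source.
# It allows one instance per owner and sends to the music sidechain bus.
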MetaSoundSource /HoneyFiberEnv/Audio/VO/IP_MSS_VO_HoneyFibre_OneShot_Parent
inputs [
 In Array {type=WaveAsset:Array}
 Time.ElapsedTime {type=Float}
 Time.GameTimeSeconds {type=Float}
 Time.PlayTime {type=Float}
 Time.TimeSinceLastPlayed {type=Float}
 UE.Source.OnPlay {type=Trigger}
]
interfaces [Fort.Time:1.0, UE.OutputFormat.Mono:1.0, UE.Source.OneShot:1.0, UE.Source:1.1]
outputs [
 "UE.OutputFormat.Mono.Audio:0" {type=Audio}
 UE.Source.OneShot.OnFinished {type=Trigger}
]
pages [Default {
  input_defaults {Time.ElapsedTime=0, Time.GameTimeSeconds=0, Time.PlayTime=0, Time.TimeSinceLastPlayed=0, UE.Source.OnPlay=null, In Array []}
  nodes [
   InitVariable.Trigger()
   InitVariable.Trigger()
   TriggerOnValueChange_0 = UE.Trigger On Value Change:Float(Value=^Time.TimeSinceLastPlayed)
   n1 = TriggerCompare.Clamp:Float(A=^Time.TimeSinceLastPlayed, Compare=@TriggerOnValueChange_0.Trigger) {also {defaults {Type=4}}}
   TriggerAnyTriggerAny2_0 = "TriggerAny.Trigger Any (2)"(In 1=TriggerCompare.Clamp:Float(A=^Time.TimeSinceLastPlayed, Compare=@n1.False) {also {defaults {B=10, Type=5}}}.True, In 0=@n1.True)
   /FortMetasound/Patches/PrintFloat(Input=^Time.TimeSinceLastPlayed, Trigger=@TriggerAnyTriggerAny2_0.Out)
   ArrayRandomGet_0 = Array.Random Get:WaveAsset:Array(In Array=^In Array, Next=@TriggerAnyTriggerAny2_0.Out)
   WavePlayer_0 = UE.Wave Player:Mono(Play=@ArrayRandomGet_0.On Next, Wave Asset=@ArrayRandomGet_0.Value) {also {version [1,1]}}
  ]
  links [{dst="@UE.Source.OneShot.OnFinished", src=WavePlayer_0.On Finished}, {dst="@UE.OutputFormat.Mono.Audio:0", src=WavePlayer_0.Out Mono}]
 }]
properties {
 assetImportData=&self:AssetImportData
 attenuationSettings=/HoneyFiberEnv/Audio/VO/AwesomeOStatue/IP_ATT_HoneyFibre_AwesomeO
 bDebug=true
 bOverrideConcurrency=true
 maxDistance=5530
 sampleRate=48000
 soundClassObject=&Fort_VO_NPC
 busSends [{audioBus=&AB_Sidechain_FortMusic_Vol, sendLevel=1}]
 concurrencyOverrides {bLimitToOwner=true, maxCount=1, resolutionRule=PreventNew}
}
