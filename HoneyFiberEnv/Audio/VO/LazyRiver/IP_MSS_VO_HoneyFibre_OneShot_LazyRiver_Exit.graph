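# A mono one-shot MetaSound Source for a HoneyFiber water park voice line, named for exiting the lazy river.
# It is a preset of the shared HoneyFibre one-shot VO parent, filling its wave array input with a single "YoureATowel" exit clip.
# It uses the water park enter/exit attenuation and the NPC VO sound class.
# A concurrency override limits playback to one instance per owner.
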MetaSoundSource /HoneyFiberEnv/Audio/VO/LazyRiver/IP_MSS_VO_HoneyFibre_OneShot_LazyRiver_Exit
preset_parent=/HoneyFiberEnv/Audio/VO/IP_MSS_VO_HoneyFibre_OneShot_Parent
inputs [
 In Array {type=WaveAsset:Array}
 UE.Source.OnPlay {type=Trigger}
]
interfaces [UE.OutputFormat.Mono:1.0, UE.Source.OneShot:1.0, UE.Source:1.1]
outputs [
 "UE.OutputFormat.Mono.Audio:0" {type=Audio}
 UE.Source.OneShot.OnFinished {type=Trigger}
]
pages [Default {input_defaults {UE.Source.OnPlay=null, In Array [/HoneyFiberEnv/Audio/VO/LazyRiver/Waves/IP_VO_HoneyFiber_WaterPark_OnExit_YoureATowel]}}]
properties {
 assetImportData=&self:AssetImportData
 attenuationSettings=&/IP_ATT_HoneyFibre_WaterParkEnterExit
 bDebug=true
 bOverrideConcurrency=true
 maxDistance=6000
 sampleRate=48000
 soundClassObject=&Fort_VO_NPC
 concurrencyOverrides {bLimitToOwner=true, maxCount=1}
}
